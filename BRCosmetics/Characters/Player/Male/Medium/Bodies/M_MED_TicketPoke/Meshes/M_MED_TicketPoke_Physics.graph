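# Physics asset for the TicketPoke medium male player body cosmetic, covering its dangling front belts, skirt panels and belt bag.
# Pelvis, spine and thighs carry simple body shapes, while overlapping capsules approximate the skirt panels.
# Belt chains hang from spine_01; front skirt chains, a left-only side skirt and the bag hang from the pelvis.
# Angular motion is free on all joints except the links along the belt chains.
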PhysicsAsset /BRCosmetics/Characters/Player/Male/Medium/Bodies/M_MED_TicketPoke/Meshes/M_MED_TicketPoke_Physics
bodies {
 MSTR_frontBelt_l [{shapeType=Capsule, length=0.2, radius=0.2, shapeName=MSTR_frontBelt_l_capsule}]
 MSTR_frontBelt_r [{shapeType=Capsule, length=0.2, radius=0.2, shapeName=MSTR_frontBelt_r_capsule}]
 MSTR_frontSkirt [
  {
   shapeType=Capsule
   length=7.3178
   radius=4.4824
   shapeName=MSTR_frontSkirt_capsule
   center=<-0.6157 -0.8363 -3.7588>
   rotation=<rot 86.4022 -90.0076 -0.0076>
  }
  {
   shapeType=Capsule
   length=7.1461
   radius=4.5217
   shapeName=MSTR_frontSkirt_capsule
   center=<-5.1463 0.0074 -3.8445>
   rotation=<rot 0.0001 0 -90.1628>
  }
 ]
 MSTR_frontSkirt_l [
  {
   shapeType=Capsule
   length=6.6706
   radius=4.4898
   shapeName=MSTR_frontSkirt_l_capsule
   center=<-1.1114 2.1596 -4.9271>
   rotation=<rot 80.7531 86.109 176.6086>
  }
  {
   shapeType=Capsule
   length=7.8221
   radius=3.9233
   shapeName=MSTR_frontSkirt_capsule
   center=<-4.7796 1.8578 -4.2826>
   rotation=<rot -0.5405 -0.4917 -78.4681>
  }
 ]
 MSTR_frontSkirt_r [
  {
   shapeType=Capsule
   length=9.8928
   radius=4
   shapeName=MSTR_frontSkirt_r_capsule
   center=<0.0002 -0.7529 -4.4265>
   rotation=<rot 89.7673 90.0187 -179.9929>
  }
  {
   shapeType=Capsule
   length=11.0058
   radius=4
   shapeName=MSTR_frontSkirt_capsule
   center=<-6.999 -0.9613 -4.4663>
   rotation=<rot 0.0093 -0.004 -89.5279>
  }
  {
   shapeType=Capsule
   length=9.8928
   radius=4
   shapeName=MSTR_frontSkirt_r_capsule
   center=<2.7502 -6.6105 -4.282>
   rotation=<rot 73.5955 -90.0228 -0.0006>
  }
 ]
 dyn_beltBag_01 [{shapeType=Sphere, radius=3.914, shapeName=dyn_beltBag_01_sphere, center=<1.5948 0.0564 0.283>}]
 dyn_frontBelt_01_l [{
   shapeType=Capsule
   length=4.71
   radius=1.3288
   shapeName=dyn_frontBelt_01_l_capsule
   center=<3.5862 -0.6486 0.4877>
   rotation=<rot 86.9475 -179.9999 -179.9999>
  }]
 dyn_frontBelt_01_r [{shapeType=Capsule, length=4.6224, radius=1.329, shapeName=dyn_frontBelt_01_l_capsule, center=<3.0729 0.2501 0.2971>, rotation=<rot 90 0 0>}]
 dyn_frontBelt_02_l [{shapeType=Capsule, length=6.1751, radius=1.6402, shapeName=dyn_frontBelt_01_l_capsule, center=<3.8948 -0.7248 0.5326>, rotation=<rot 90 0 0>}]
 dyn_frontBelt_02_r [{shapeType=Capsule, length=4.9855, radius=1.4999, shapeName=dyn_frontBelt_01_l_capsule, center=<3.4945 0.2747 0.233>, rotation=<rot 90 0 0>}]
 dyn_frontBelt_03_l [{
   shapeType=Capsule
   length=7.4352
   radius=1.8579
   shapeName=dyn_frontBelt_01_l_capsule
   center=<4.3527 -0.3652 0.5548>
   rotation=<rot 89.5669 -178.2869 -179.9998>
  }]
 dyn_frontBelt_03_r [{
   shapeType=Capsule
   length=5.892
   radius=1.4923
   shapeName=dyn_frontBelt_01_l_capsule
   center=<4.4273 0.663 0.2407>
   rotation=<rot 88.0251 -178.0571 -179.9999>
  }]
 dyn_frontBelt_04_l [{
   shapeType=Capsule
   length=8.0935
   radius=1.9065
   shapeName=dyn_frontBelt_01_l_capsule
   center=<4.6695 0.0294 0.1697>
   rotation=<rot 85.7174 1.0098 -0.2687>
  }]
 dyn_frontSkirt_01 [
  {
   shapeType=Capsule
   length=8.2461
   radius=4.5822
   shapeName=dyn_frontSkirt_01_capsule
   center=<6.3584 -0.9462 -4.5348>
   rotation=<rot -0.0001 -179.7299 -84.7696>
  }
  {
   shapeType=Capsule
   length=8.5756
   radius=4.5822
   shapeName=dyn_frontSkirt_01_capsule
   center=<1.9311 -0.991 -4.5481>
   rotation=<rot -0.0001 -179.7299 -84.7696>
  }
  {
   shapeType=Capsule
   length=8.0725
   radius=2.4656
   shapeName=dyn_frontSkirt_01_capsule
   center=<5.7424 -7.2758 -3.5797>
   rotation=<rot 0.0001 0 -102.8663>
  }
  {
   shapeType=Capsule
   length=9.6012
   radius=3.7746
   shapeName=dyn_frontSkirt_02_capsule
   center=<6.1888 4.8471 -4.2881>
   rotation=<rot 2.848 -170.8809 -109.2338>
  }
 ]
 dyn_frontSkirt_01_l [
  {
   shapeType=Capsule
   length=8.3442
   radius=4.6793
   shapeName=dyn_frontSkirt_01_l_capsule
   center=<8.1031 2.0696 -5.3301>
   rotation=<rot 0.0005 -1.7606 -75.6353>
  }
  {
   shapeType=Capsule
   length=8.679
   radius=3.9586
   shapeName=dyn_frontSkirt_01_l_capsule
   center=<3.7864 2.1879 -4.6522>
   rotation=<rot 0.0005 -1.7605 -71.876>
  }
  {
   shapeType=Capsule
   length=8.679
   radius=3.9586
   shapeName=dyn_frontSkirt_01_l_capsule
   center=<10.8044 2.3301 -4.4979>
   rotation=<rot 0.0005 -1.7605 -71.876>
  }
 ]
 dyn_frontSkirt_01_r [
  {
   shapeType=Capsule
   length=9.7307
   radius=4.6495
   shapeName=dyn_frontSkirt_02_r_capsule
   center=<4.8286 -3.8631 -3.9957>
   rotation=<rot -2.5791 -179.1659 -77.7498>
  }
  {
   shapeType=Capsule
   length=8.7406
   radius=3.9838
   shapeName=dyn_frontSkirt_02_r_capsule
   center=<9.6162 -7.0134 -4.1631>
   rotation=<rot -2.579 -179.1658 -74.1366>
  }
 ]
 dyn_frontSkirt_02 [
  {
   shapeType=Capsule
   length=9.1428
   radius=3.7746
   shapeName=dyn_frontSkirt_02_c_2_capsule
   center=<4.9064 -0.2032 -3.8471>
   rotation=<rot -0.2885 179.9181 -91.8041>
  }
  {
   shapeType=Capsule
   length=9.7039
   radius=2.8099
   shapeName=dyn_frontSkirt_02_c_3_capsule
   center=<7.2351 0.0071 -2.2607>
   rotation=<rot 0 0 -90.1801>
  }
  {
   shapeType=Capsule
   length=8.9146
   radius=3.7746
   shapeName=dyn_frontSkirt_02_c_1_capsule
   center=<1.0916 -0.7116 -4.0639>
   rotation=<rot -0.2885 179.9181 -91.8041>
  }
  {
   shapeType=Capsule
   length=8.9077
   radius=2.3975
   shapeName=dyn_frontSkirt_02_l_2_capsule
   center=<6.9425 -8.9145 -4.2495>
   rotation=<rot 9.0663 162.9227 -82.6219>
  }
  {
   shapeType=Capsule
   length=8.9077
   radius=2.3975
   shapeName=dyn_frontSkirt_02_l_1_capsule
   center=<2.26 -7.2606 -3.6264>
   rotation=<rot 9.0663 162.9227 -77.4133>
  }
  {
   shapeType=Capsule
   length=10.2292
   radius=2.8543
   shapeName=dyn_frontSkirt_02_r_2_capsule
   center=<6.2929 6.711 -3.097>
   rotation=<rot 17.4321 -167.5803 -100.7204>
  }
  {
   shapeType=Capsule
   length=9.1191
   radius=3.0124
   shapeName=dyn_frontSkirt_02_r_1_capsule
   center=<2.1772 6.2353 -2.9262>
   rotation=<rot 16.9285 -170.314 -103.4065>
  }
  {
   shapeType=Capsule
   length=9.734
   radius=2.3975
   shapeName=dyn_frontSkirt_02_r_3_capsule
   center=<-2.5253 5.9287 -2.1211>
   rotation=<rot 16.9285 -170.314 -103.4065>
  }
 ]
 dyn_frontSkirt_02_l [{
   shapeType=Capsule
   length=9.8
   radius=5.8476
   shapeName=dyn_frontSkirt_02_l_capsule
   center=<5.7286 5.1457 -7.6357>
   rotation=<rot -0.5494 -4.9095 -71.7174>
  }]
 dyn_frontSkirt_02_r [
  {
   shapeType=Capsule
   length=8.9694
   radius=5.4108
   shapeName=dyn_frontSkirt_02_r_capsule
   center=<4.8185 -2.7159 -4.1431>
   rotation=<rot -2.5788 -179.1657 -93.6247>
  }
  {
   shapeType=Capsule
   length=8.7406
   radius=3.9838
   shapeName=dyn_frontSkirt_02_r_capsule
   center=<6.1323 -5.9512 -3.9059>
   rotation=<rot -2.5789 -179.1658 -74.6343>
  }
 ]
 dyn_sideSkirt_01_l [{
   shapeType=Capsule
   length=4.01
   radius=4.9461
   shapeName=dyn_sideSkirt_01_l_capsule
   center=<4.2481 3.6684 -7.4946>
   rotation=<rot -0.0021 0.0173 153.37>
  }]
 dyn_sideSkirt_02_l [{
   shapeType=Capsule
   length=4.8631
   radius=5.4353
   shapeName=dyn_sideSkirt_02_l_capsule
   center=<3.8821 2.5266 -7.0037>
   rotation=<rot -0.001 0.0079 113.544>
  }]
 pelvis [{
   shapeType=Capsule
   length=11.0302
   radius=14.5
   shapeName=pelvis_capsule
   center=<4.6682 -4.3315 -1.266>
   rotation=<rot 87.4463 -0.0005 6.5872>
  }]
 spine_01 [{shapeType=Sphere, radius=10.9601, shapeName=spine_01_sphere, center=<0.4207 -0.534 -2.7348>}]
 thigh_l [{
   shapeType=Capsule
   length=35.8157
   radius=10.5851
   shapeName=thigh_l_capsule
   center=<-15.0964 -0.5183 1.0874>
   rotation=<rot -10.6275 -91.4413 88.144>
  }]
 thigh_r [{
   shapeType=Capsule
   length=32.9727
   radius=10.0553
   shapeName=thigh_r_capsule
   center=<14.0226 1.7942 0.5702>
   rotation=<rot 7.2567 -94.5982 92.8403>
  }]
}
constraints [
 {
  bone1Name=thigh_l
  bone2Name=pelvis
  swing1LimitDegrees=45
  swing1Motion=Free
  swing2LimitDegrees=45
  swing2Motion=Free
  twistLimitDegrees=45
  twistMotion=Free
 }
 {
  bone1Name=thigh_r
  bone2Name=pelvis
  swing1LimitDegrees=45
  swing1Motion=Free
  swing2LimitDegrees=45
  swing2Motion=Free
  twistLimitDegrees=45
  twistMotion=Free
 }
 {
  bone1Name=MSTR_frontBelt_l
  bone2Name=spine_01
  swing1LimitDegrees=45
  swing1Motion=Free
  swing2LimitDegrees=45
  swing2Motion=Free
  twistLimitDegrees=45
  twistMotion=Free
 }
 {bone1Name=dyn_frontBelt_01_l, bone2Name=MSTR_frontBelt_l, swing1LimitDegrees=45, swing2LimitDegrees=45, twistLimitDegrees=0}
 {bone1Name=dyn_frontBelt_02_l, bone2Name=dyn_frontBelt_01_l, swing1LimitDegrees=45, swing2LimitDegrees=45, twistLimitDegrees=0}
 {bone1Name=dyn_frontBelt_03_l, bone2Name=dyn_frontBelt_02_l, swing1LimitDegrees=45, swing2LimitDegrees=45, twistLimitDegrees=0}
 {bone1Name=dyn_frontBelt_04_l, bone2Name=dyn_frontBelt_03_l, swing1LimitDegrees=45, swing2LimitDegrees=45, twistLimitDegrees=0}
 {
  bone1Name=MSTR_frontBelt_r
  bone2Name=spine_01
  swing1LimitDegrees=45
  swing1Motion=Free
  swing2LimitDegrees=45
  swing2Motion=Free
  twistLimitDegrees=45
  twistMotion=Free
 }
 {bone1Name=dyn_frontBelt_01_r, bone2Name=MSTR_frontBelt_r, swing1LimitDegrees=45, swing2LimitDegrees=25, twistLimitDegrees=0}
 {bone1Name=dyn_frontBelt_02_r, bone2Name=dyn_frontBelt_01_r, swing1LimitDegrees=45, swing2LimitDegrees=35, twistLimitDegrees=0}
 {bone1Name=dyn_frontBelt_03_r, bone2Name=dyn_frontBelt_02_r, swing1LimitDegrees=45, swing2LimitDegrees=45, twistLimitDegrees=0}
 {
  bone1Name=dyn_frontSkirt_01
  bone2Name=MSTR_frontSkirt
  swing1LimitDegrees=45
  swing1Motion=Free
  swing2LimitDegrees=45
  swing2Motion=Free
  twistLimitDegrees=45
  twistMotion=Free
 }
 {
  bone1Name=dyn_frontSkirt_02
  bone2Name=dyn_frontSkirt_01
  swing1LimitDegrees=45
  swing1Motion=Free
  swing2LimitDegrees=45
  swing2Motion=Free
  twistLimitDegrees=45
  twistMotion=Free
 }
 {
  bone1Name=dyn_frontSkirt_01_l
  bone2Name=MSTR_frontSkirt_l
  swing1LimitDegrees=45
  swing1Motion=Free
  swing2LimitDegrees=45
  swing2Motion=Free
  twistLimitDegrees=45
  twistMotion=Free
 }
 {
  bone1Name=dyn_frontSkirt_02_l
  bone2Name=dyn_frontSkirt_01_l
  swing1LimitDegrees=45
  swing1Motion=Free
  swing2LimitDegrees=45
  swing2Motion=Free
  twistLimitDegrees=45
  twistMotion=Free
 }
 {
  bone1Name=dyn_frontSkirt_01_r
  bone2Name=MSTR_frontSkirt_r
  swing1LimitDegrees=45
  swing1Motion=Free
  swing2LimitDegrees=45
  swing2Motion=Free
  twistLimitDegrees=45
  twistMotion=Free
 }
 {
  bone1Name=dyn_frontSkirt_02_r
  bone2Name=dyn_frontSkirt_01_r
  swing1LimitDegrees=45
  swing1Motion=Free
  swing2LimitDegrees=45
  swing2Motion=Free
  twistLimitDegrees=45
  twistMotion=Free
 }
 {
  bone1Name=MSTR_frontSkirt
  bone2Name=pelvis
  swing1LimitDegrees=45
  swing1Motion=Free
  swing2LimitDegrees=45
  swing2Motion=Free
  twistLimitDegrees=45
  twistMotion=Free
 }
 {
  bone1Name=MSTR_frontSkirt_l
  bone2Name=pelvis
  swing1LimitDegrees=45
  swing1Motion=Free
  swing2LimitDegrees=45
  swing2Motion=Free
  twistLimitDegrees=45
  twistMotion=Free
 }
 {
  bone1Name=MSTR_frontSkirt_r
  bone2Name=pelvis
  swing1LimitDegrees=45
  swing1Motion=Free
  swing2LimitDegrees=45
  swing2Motion=Free
  twistLimitDegrees=45
  twistMotion=Free
 }
 {
  bone1Name=dyn_beltBag_01
  bone2Name=pelvis
  swing1LimitDegrees=45
  swing1Motion=Free
  swing2LimitDegrees=45
  swing2Motion=Free
  twistLimitDegrees=45
  twistMotion=Free
 }
 {
  bone1Name=dyn_sideSkirt_01_l
  bone2Name=pelvis
  swing1LimitDegrees=45
  swing1Motion=Free
  swing2LimitDegrees=45
  swing2Motion=Free
  twistLimitDegrees=45
  twistMotion=Free
 }
 {
  bone1Name=dyn_sideSkirt_02_l
  bone2Name=dyn_sideSkirt_01_l
  swing1LimitDegrees=45
  swing1Motion=Free
  swing2LimitDegrees=45
  swing2Motion=Free
  twistLimitDegrees=45
  twistMotion=Free
 }
 {
  bone1Name=spine_01
  bone2Name=pelvis
  swing1LimitDegrees=45
  swing1Motion=Free
  swing2LimitDegrees=45
  swing2Motion=Free
  twistLimitDegrees=45
  twistMotion=Free
 }
]
