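# Physics asset for the face accessory of the TicketPoke male medium character, covering two dangling mask bone chains.
# Capsule chains hang from left and right MSTR_mask roots on the head, three dyn_mask segments left and four right.
# Chain joints have tight twist limits and swing limits that widen down the chain.
# A head sphere and neck, clavicle and upper spine capsules form the body shapes, linked by free angular constraints.
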PhysicsAsset /BRCosmetics/Characters/Player/Male/Medium/Bodies/M_MED_TicketPoke/Meshes/Parts/M_MED_TicketPoke_FaceAcc_Physics
bodies {
 MSTR_mask_01_l [{shapeType=Capsule, length=0.505, radius=0.505, shapeName=MSTR_mask_01_l_capsule, rotation=<rot 0 -90 0>}]
 MSTR_mask_01_r [{shapeType=Capsule, length=0.505, radius=0.505, shapeName=MSTR_mask_01_r_capsule, rotation=<rot 0 -90 0>}]
 clavicle_l [{
   shapeType=Capsule
   length=14.0802
   radius=9.9498
   shapeName=clavicle_l_capsule
   center=<14.9227 -2.0734 0.893>
   rotation=<rot -0.0001 -90 93.6169>
  }]
 clavicle_r [{
   shapeType=Capsule
   length=13.0793
   radius=10.5188
   shapeName=clavicle_r_capsule
   center=<-16.3458 2.5476 0.1205>
   rotation=<rot -0.1657 -92.4643 94.2094>
  }]
 dyn_mask_01_l [{shapeType=Capsule, length=2.9889, radius=1.5211, shapeName=dyn_mask_01_l_capsule, center=<3.3942 0 0.2943>, rotation=<rot 0 -90 96.0442>}]
 dyn_mask_01_r [{
   shapeType=Capsule
   length=2.8546
   radius=1.6554
   shapeName=dyn_mask_01_r_capsule
   center=<3.7223 -0.1859 0.3789>
   rotation=<rot 0.0001 -90 86.4634>
  }]
 dyn_mask_02_l [{shapeType=Capsule, length=4.3648, radius=2.1352, shapeName=dyn_mask_02_l_capsule, center=<3.6587 0 0.8573>, rotation=<rot 0 -90 91.7469>}]
 dyn_mask_02_r [{
   shapeType=Capsule
   length=4.8879
   radius=2.1139
   shapeName=dyn_mask_02_r_capsule
   center=<5.0427 -0.0725 0.7057>
   rotation=<rot 0.5244 -86.9034 -92.9699>
  }]
 dyn_mask_03_l [{
   shapeType=Capsule
   length=6.885
   radius=1.5625
   shapeName=dyn_mask_03_l_capsule
   center=<4.4645 0.2125 0.7602>
   rotation=<rot -0.1324 94.6938 -93.7398>
  }]
 dyn_mask_03_r [{
   shapeType=Capsule
   length=6.4363
   radius=1.9364
   shapeName=dyn_mask_03_r_capsule
   center=<6.9366 0.0081 1.2607>
   rotation=<rot 0.0362 -87.682 -86.538>
  }]
 dyn_mask_04_r [{
   shapeType=Capsule
   length=6.4994
   radius=2.0169
   shapeName=dyn_mask_04_r_capsule
   center=<6.6366 0.0353 0.2297>
   rotation=<rot 0.0901 -90.3797 -89.0838>
  }]
 head [{shapeType=Sphere, radius=10.7652, shapeName=head_sphere, center=<7.3536 2.3552 0.0032>}]
 neck_01 [{shapeType=Capsule, length=0.8576, radius=7.6474, shapeName=neck_01_capsule, center=<5.0446 0.4083 0>, rotation=<rot 0 -90 0>}]
 spine_04 [{shapeType=Capsule, length=11.863, radius=13.142, shapeName=spine_05_capsule, center=<0.1827 1.3732 0>, rotation=<rot 0 -90 0>}]
 spine_05 [
  {shapeType=Capsule, length=14.4569, radius=11.6283, shapeName=spine_05_capsule, center=<2.4077 -0.8017 0>, rotation=<rot 0 -90 0>}
  {shapeType=Capsule, length=10.4179, radius=7.5871, shapeName=spine_05_capsule, center=<10.5858 0.1313 0>, rotation=<rot 0 -90 0>}
 ]
}
constraints [
 {
  bone1Name=head
  bone2Name=neck_01
  swing1LimitDegrees=45
  swing1Motion=Free
  swing2LimitDegrees=45
  swing2Motion=Free
  twistLimitDegrees=45
  twistMotion=Free
 }
 {
  bone1Name=clavicle_l
  bone2Name=spine_05
  swing1LimitDegrees=45
  swing1Motion=Free
  swing2LimitDegrees=45
  swing2Motion=Free
  twistLimitDegrees=45
  twistMotion=Free
 }
 {
  bone1Name=clavicle_r
  bone2Name=spine_05
  swing1LimitDegrees=45
  swing1Motion=Free
  swing2LimitDegrees=45
  swing2Motion=Free
  twistLimitDegrees=45
  twistMotion=Free
 }
 {
  bone1Name=neck_01
  bone2Name=spine_05
  swing1LimitDegrees=45
  swing1Motion=Free
  swing2LimitDegrees=45
  swing2Motion=Free
  twistLimitDegrees=45
  twistMotion=Free
 }
 {
  bone1Name=MSTR_mask_01_l
  bone2Name=head
  swing1LimitDegrees=45
  swing1Motion=Free
  swing2LimitDegrees=45
  swing2Motion=Free
  twistLimitDegrees=45
  twistMotion=Free
 }
 {bone1Name=dyn_mask_01_l, bone2Name=MSTR_mask_01_l, swing1LimitDegrees=25, swing2LimitDegrees=25, twistLimitDegrees=5}
 {bone1Name=dyn_mask_02_l, bone2Name=dyn_mask_01_l, swing1LimitDegrees=30, swing2LimitDegrees=30, twistLimitDegrees=5}
 {bone1Name=dyn_mask_03_l, bone2Name=dyn_mask_02_l, swing1LimitDegrees=35, swing2LimitDegrees=35, twistLimitDegrees=5}
 {
  bone1Name=MSTR_mask_01_r
  bone2Name=head
  swing1LimitDegrees=45
  swing1Motion=Free
  swing2LimitDegrees=45
  swing2Motion=Free
  twistLimitDegrees=45
  twistMotion=Free
 }
 {bone1Name=dyn_mask_01_r, bone2Name=MSTR_mask_01_r, swing1LimitDegrees=25, swing2LimitDegrees=25, twistLimitDegrees=5}
 {bone1Name=dyn_mask_02_r, bone2Name=dyn_mask_01_r, swing1LimitDegrees=30, swing2LimitDegrees=30, twistLimitDegrees=5}
 {bone1Name=dyn_mask_03_r, bone2Name=dyn_mask_02_r, swing1LimitDegrees=35, swing2LimitDegrees=35, twistLimitDegrees=5}
 {bone1Name=dyn_mask_04_r, bone2Name=dyn_mask_03_r, swing1LimitDegrees=35, swing2LimitDegrees=35, twistLimitDegrees=5}
 {
  bone1Name=spine_05
  bone2Name=spine_04
  swing1LimitDegrees=45
  swing1Motion=Free
  swing2LimitDegrees=45
  swing2Motion=Free
  twistLimitDegrees=45
  twistMotion=Free
 }
]
