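# Post-copy Anim Blueprint for the TicketPoke character's face accessory, driving its secondary motion.
# A Control Rig adjusts the parent-mesh pose, then RigidBody simulation runs with state-based gravity, plus additive mask-bone rotation during skydive or parachute.
# SplineFunc, true with qualifying backbling or disabled rigid bodies, feeds the rig and bypasses physics and mask offsets.
# Per-state RigidBody settings are computed each update but not visibly wired into the node.
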AnimBlueprint /BRCosmetics/Characters/Player/Male/Medium/Bodies/M_MED_TicketPoke/Meshes/Parts/M_MED_TicketPoke_FaceAcc_AnimBP
target_skeleton=&/M_MED_TicketPoke_FaceAcc_Skeleton
graphs [
 AnimGraph {nodes [
   n1 = RigidBody!(
    ComponentPose=Local To Component(LocalPose=Use cached pose 'Input pose'())
    Alpha
    OverrideWorldGravity:Vector=Get Gravity Override()
    ExternalForce
   ) -> (Pose) {
    bOverrideWorldGravity=true
    overridePhysicsAsset=&/M_MED_TicketPoke_FaceAcc_Physics
    simulationSpace=BaseBoneSpace
    alphaBoolBlend {blendOption=Linear}
    baseBoneRef {boneName=spine_05}
    componentAppliedLinearAccClamp=<0 0 0>
    overrideWorldGravity {z=-980}
    simSpaceSettings {maxAngularAcceleration=40, maxAngularVelocity=10, maxLinearAcceleration=1000, maxLinearVelocity=300, velocityScaleZ=0.5, worldAlpha=1}
   }
   n2 = Get Alpha Skydive Mode() -> (Alpha Skydive Mode:double)
   n3 = "Transform (Modify) Bone - Bone: MSTR_mask_01_r"(
    ComponentPose="Transform (Modify) Bone - Bone: MSTR_mask_01_l"(ComponentPose=@n1, Alpha=@n2, Translation="0,0,0", Rotation="12,0,0", Scale="1,1,1") {rotationMode=BMM_Additive, rotationSpace=BCS_BoneSpace, boneToModify {boneName=MSTR_mask_01_l}, rotation {pitch=12}}
    Alpha=@n2
    Translation="0,0,0"
    Rotation="12,0,0"
    Scale="1,1,1"
   ) {rotationMode=BMM_Additive, rotationSpace=BCS_BoneSpace, boneToModify {boneName=MSTR_mask_01_r}, rotation {pitch=12}}
   Output Pose(Result=Blend Poses by bool(
     BlendPose_0=Use cached pose 'Input pose'()
     BlendPose_1=Component To Local(ComponentPose=@n3)
     bActiveValue=Get SplineFunc() -> (SplineFunc:bool)
    ) {blendType=Linear}) {layerGroup=None}
   n4 = Control Rig!(
    Source=Copy Pose From Mesh() {bUseAttachedParent=true}
    Alpha
    BackblingOn:bool=Get Is backbling on() -> (Is backbling on:bool)
    SplineFunc:bool=Get SplineFunc()
   ) -> (Pose) {
    alphaBoolBlend {blendOption=Linear}
    controlRigAssetReference {blueprintRigClass=&/M_MED_TicketPoke_FaceAcc_CtrlRig.M_MED_TicketPoke_FaceAcc_CtrlRig_C}
   }
   Save cached pose 'Input pose'(Pose=@n4) {cacheName=Input pose}
  ]}
 EventGraph {nodes [
   Event BlueprintUpdateAnimation()
   > n1 = "Gravity Override State Function (Nativized)"!(
    execute
    CharacterPartAnimInstance:Custom Character Part Anim Instance Object Reference=Self-Reference()
    GravityOverrideParameterSet:Fort Gravity Override Parameter Set Structure=Get Gravity Override Params() -> (Gravity Override Params:Fort Gravity Override Parameter Set Structure)
    RigidBodyAnimNodeParameterSet:Fort Rigid Body Anim Node Parameter Set Structure=Get RBAN Params() -> (RBAN Params:Fort Rigid Body Anim Node Parameter Set Structure)
    DebugDraw:bool=false
    GameTimeInSeconds:float=GetGameTimeInSeconds()
    DebugJointName:Name=root
    StateLogicLayerName:Name=none
   ) -> (
    then
    OutGravityOverride:Vector
    OutCurrentState:Name=None
    OutGravityOverrideParameters:Fort Gravity Override Parameters Structure
    OutRigidBodyAnimNodeParameters:Fort Rigid Body Anim Node Parameters Structure
   )
   > Set Gravity Override(Gravity Override=@n1)
   > Set RBAN Settings Out(RBAN Settings Out=@n1.OutRigidBodyAnimNodeParameters)
   n2 = CheckForBackbling()
   > Set Is backbling on(Is backbling on:bool=OR Boolean!(A=@n2, B=@n2.bIsBackblingACape, C:bool=@n2.bIsBackblingLong) -> (ReturnValue)) -> (Output_Get:bool)
   n3 = Get bRigidBodyEnabled()
   > Set SplineFunc(SplineFunc:bool=OR Boolean!(A=Get Is backbling on(), B=NOT Boolean(A=@n3)) -> (ReturnValue)) -> (Output_Get:bool)
   > n4 = Branch(Condition=OR Boolean!(A=Get bIsSkydiveDiveMode(), B=Get bIsSkydiveFloating(), C:bool=Get bIsParachuteOpened()) -> (ReturnValue))
   > Set Alpha Skydive Mode(Alpha Skydive Mode:double=1) -> (Output_Get:double)
   Set Alpha Skydive Mode(execute=@n4.else, Alpha Skydive Mode=0)
  ]}
]
variables [
 Alpha Skydive Mode {type=number}
 Gravity Override {type=Vector}
 Gravity Override Params {
  type=FortGravityOverrideParameterSet
  default=(Skydive_GoingUp=(GravityOverride=(WindFrequency=(X=1.500000,Y=7.000000,Z=8.000000),WindAmplitude=(X=200.000000,Y=-500.000000,Z=400.000000),GravityOverride=(X=0.000000,Y=-300.000000,Z=100.000000),JointName="root")),Skydive_FloatingDown=(GravityOverride=(WindFrequency=(X=1.500000,Y=7.000000,Z=8.000000),WindAmplitude=(X=200.000000,Y=-500.000000,Z=400.000000),WindOffset=(X=0.000000,Y=-600.000000,Z=400.000000),GravityOverride=(X=0.000000,Y=-700.000000,Z=600.000000),JointName="root")),Skydive_DivingDown=(GravityOverride=(WindFrequency=(X=1.500000,Y=12.000000,Z=10.000000),WindAmplitude=(X=50.000000,Y=300.000000,Z=100.000000),GravityOverride=(X=-1500.000000,Y=0.000000,Z=0.000000),JointName="pelvis")),Skydive_Parachuting=(GravityOverride=(WindFrequency=(X=1.500000,Y=7.000000,Z=8.000000),WindAmplitude=(X=200.000000,Y=-500.000000,Z=400.000000),GravityOverride=(X=0.000000,Y=-800.000000,Z=-980.000000),JointName="root")),OnGround_Standing=(GravityOverride=(GravityOverride=(X=0.000000,Y=0.000000,Z=-980.000000),JointName="root")),OnGround_Moving=(GravityOverride=(GravityOverride=(X=0.000000,Y=0.000000,Z=-980.000000),JointName="root")),OnGround_Moving_HighRotationRate=(GravityOverride=(GravityOverride=(X=0.000000,Y=0.000000,Z=-980.000000),JointName="root")),OnGround_Crouch_Moving=(GravityOverride=(GravityOverride=(X=0.000000,Y=0.000000,Z=-980.000000),JointName="root")),Emote_or_Melee=(GravityOverride=(GravityOverride=(X=0.000000,Y=0.000000,Z=-980.000000),JointName="root")),Swimming=(GravityOverride=(GravityOverride=(X=-800.000000,Y=0.000000,Z=0.000000),JointName="pelvis")),Front_End=(GravityOverride=(GravityOverride=(X=0.000000,Y=0.000000,Z=-980.000000),JointName="root")),DBNO=(GravityOverride=(GravityOverride=(X=-980.000000,Y=400.000000,Z=0.000000),JointName="pelvis")),DBNO_Carried=(GravityOverride=(GravityOverride=(X=-980.000000,Y=500.000000,Z=-500.000000),JointName="pelvis")),OnGround_Falling=(GravityOverride=(GravityOverride=(X=0.000000,Y=0.000000,Z=-980.000000),JointName="root")),Ziplining=(GravityOverride=(GravityOverride=(X=0.000000,Y=0.000000,Z=-980.000000),JointName="root")),Skydive_Surfing=(GravityOverride=(WindFrequency=(X=1.500000,Y=7.000000,Z=8.000000),WindAmplitude=(X=200.000000,Y=-500.000000,Z=400.000000),GravityOverride=(X=500.000000,Y=-300.000000,Z=-980.000000),JointName="root")),Riding_Standing=(GravityOverride=(GravityOverride=(X=0.000000,Y=-350.000000,Z=-980.000000),JointName="root")),Riding_Moving=(GravityOverride=(GravityOverride=(X=0.000000,Y=-450.000000,Z=-450.000000),JointName="root")),Riding_Moving_Targeting=(GravityOverride=(GravityOverride=(X=0.000000,Y=-200.000000,Z=-980.000000),JointName="root")),Tactical_Sprint=(GravityOverride=(GravityOverride=(X=0.000000,Y=-450.000000,Z=-500.000000),JointName="root")),Rail_Grinding=(GravityOverride=(GravityOverride=(X=0.000000,Y=0.000000,Z=-980.000000),JointName="root")),)
 }
 Is backbling on {type=boolean}
 RBAN Params {
  type=FortRigidBodyAnimNodeParameterSet
  default=(Skydive_GoingUp=(RigidBodyAnimNode=(ComponentLinearAccScale=(X=0.000000,Y=-0.500000,Z=0.000000),ComponentLinearVelScale=(X=0.000000,Y=-0.500000,Z=0.000000),SimSpaceSettings=(WorldAlpha=1.000000,VelocityScaleZ=0.000000,DampingAlpha=0.000000,MaxLinearVelocity=300.000000,MaxAngularVelocity=4.000000,MaxLinearAcceleration=1200.000000,MaxAngularAcceleration=16.000000,ExternalLinearDragV=(X=0.000000,Y=0.000000,Z=-1.000000)))),Skydive_FloatingDown=(RigidBodyAnimNode=(ComponentLinearAccScale=(X=0.000000,Y=0.500000,Z=0.000000),ComponentLinearVelScale=(X=0.000000,Y=0.500000,Z=0.000000),SimSpaceSettings=(WorldAlpha=1.000000,VelocityScaleZ=0.500000,DampingAlpha=0.000000,MaxLinearVelocity=700.000000,MaxAngularVelocity=8.000000,MaxLinearAcceleration=2800.000000,MaxAngularAcceleration=32.000000,ExternalLinearDragV=(X=0.500000,Y=0.500000,Z=0.500000)))),Skydive_DivingDown=(RigidBodyAnimNode=(ComponentLinearAccScale=(X=0.100000,Y=0.000000,Z=0.000000),ComponentLinearVelScale=(X=0.100000,Y=0.000000,Z=0.000000),SimSpaceSettings=(WorldAlpha=1.000000,VelocityScaleZ=0.500000,DampingAlpha=0.000000,MaxLinearVelocity=700.000000,MaxAngularVelocity=8.000000,MaxLinearAcceleration=2800.000000,MaxAngularAcceleration=32.000000,ExternalLinearDragV=(X=0.500000,Y=0.500000,Z=0.500000)))),Skydive_Parachuting=(RigidBodyAnimNode=(ComponentLinearAccScale=(X=0.000000,Y=0.100000,Z=0.300000),ComponentLinearVelScale=(X=0.000000,Y=0.100000,Z=0.300000),SimSpaceSettings=(WorldAlpha=1.000000,VelocityScaleZ=0.500000,DampingAlpha=0.000000,MaxLinearVelocity=700.000000,MaxAngularVelocity=8.000000,MaxLinearAcceleration=2800.000000,MaxAngularAcceleration=32.000000,ExternalLinearDragV=(X=0.500000,Y=0.500000,Z=0.500000)))),OnGround_Standing=(RigidBodyAnimNode=(ComponentLinearAccScale=(X=0.300000,Y=0.000000,Z=0.000000),ComponentLinearVelScale=(X=0.300000,Y=0.000000,Z=0.000000),SimSpaceSettings=(WorldAlpha=1.000000,VelocityScaleZ=0.500000,DampingAlpha=0.000000,MaxLinearVelocity=500.000000,MaxAngularVelocity=6.000000,MaxLinearAcceleration=2000.000000,MaxAngularAcceleration=24.000000,ExternalLinearDragV=(X=0.200000,Y=0.200000,Z=0.200000)))),OnGround_Moving=(RigidBodyAnimNode=(ComponentLinearAccScale=(X=0.300000,Y=0.000000,Z=0.000000),ComponentLinearVelScale=(X=0.300000,Y=0.000000,Z=0.000000),SimSpaceSettings=(WorldAlpha=1.000000,VelocityScaleZ=0.500000,DampingAlpha=0.000000,MaxLinearVelocity=500.000000,MaxAngularVelocity=6.000000,MaxLinearAcceleration=2000.000000,MaxAngularAcceleration=24.000000,ExternalLinearDragV=(X=0.200000,Y=0.200000,Z=0.200000)))),OnGround_Moving_HighRotationRate=(RigidBodyAnimNode=(SimSpaceSettings=(WorldAlpha=1.000000,VelocityScaleZ=0.500000,DampingAlpha=0.000000,MaxLinearVelocity=500.000000,MaxAngularVelocity=6.000000,MaxLinearAcceleration=2000.000000,MaxAngularAcceleration=24.000000,ExternalLinearDragV=(X=0.200000,Y=0.200000,Z=0.200000)))),OnGround_Crouch_Moving=(RigidBodyAnimNode=(SimSpaceSettings=(WorldAlpha=1.000000,VelocityScaleZ=0.500000,DampingAlpha=0.000000,MaxLinearVelocity=400.000000,MaxAngularVelocity=8.000000,MaxLinearAcceleration=3200.000000,MaxAngularAcceleration=32.000000,ExternalLinearDragV=(X=0.500000,Y=0.500000,Z=0.500000)))),Emote_or_Melee=(RigidBodyAnimNode=(SimSpaceSettings=(WorldAlpha=1.000000,VelocityScaleZ=0.500000,DampingAlpha=0.000000,MaxLinearVelocity=700.000000,MaxAngularVelocity=8.000000,MaxLinearAcceleration=2800.000000,MaxAngularAcceleration=32.000000,ExternalLinearDragV=(X=0.500000,Y=0.500000,Z=0.500000)))),Swimming=(RigidBodyAnimNode=(SimSpaceSettings=(WorldAlpha=1.000000,VelocityScaleZ=0.500000,DampingAlpha=0.000000,MaxLinearVelocity=700.000000,MaxAngularVelocity=8.000000,MaxLinearAcceleration=2800.000000,MaxAngularAcceleration=32.000000,ExternalLinearDragV=(X=0.500000,Y=0.500000,Z=0.500000)))),Front_End=(RigidBodyAnimNode=(SimSpaceSettings=(WorldAlpha=1.000000,VelocityScaleZ=0.500000,DampingAlpha=0.000000,MaxLinearVelocity=200.000000,MaxAngularVelocity=2.500000,MaxLinearAcceleration=800.000000,MaxAngularAcceleration=10.000000))),DBNO=(RigidBodyAnimNode=(SimSpaceSettings=(WorldAlpha=1.000000,VelocityScaleZ=0.500000,DampingAlpha=0.000000,MaxLinearVelocity=700.000000,MaxAngularVelocity=8.000000,MaxLinearAcceleration=2800.000000,MaxAngularAcceleration=32.000000,ExternalLinearDragV=(X=0.500000,Y=0.500000,Z=0.500000)))),DBNO_Carried=(RigidBodyAnimNode=(SimSpaceSettings=(WorldAlpha=1.000000,VelocityScaleZ=0.500000,DampingAlpha=0.000000,MaxLinearVelocity=700.000000,MaxAngularVelocity=8.000000,MaxLinearAcceleration=2800.000000,MaxAngularAcceleration=32.000000,ExternalLinearDragV=(X=0.500000,Y=0.500000,Z=0.500000)))),OnGround_Falling=(RigidBodyAnimNode=(SimSpaceSettings=(WorldAlpha=1.000000,VelocityScaleZ=0.300000,DampingAlpha=0.000000,MaxLinearVelocity=700.000000,MaxAngularVelocity=8.000000,MaxLinearAcceleration=2800.000000,MaxAngularAcceleration=32.000000))),Ziplining=(RigidBodyAnimNode=(SimSpaceSettings=(WorldAlpha=1.000000,VelocityScaleZ=0.500000,DampingAlpha=0.000000,MaxLinearVelocity=700.000000,MaxAngularVelocity=8.000000,MaxLinearAcceleration=2800.000000,MaxAngularAcceleration=32.000000,ExternalLinearDragV=(X=0.500000,Y=0.500000,Z=0.500000)))),Skydive_Surfing=(RigidBodyAnimNode=(SimSpaceSettings=(WorldAlpha=1.000000,VelocityScaleZ=0.500000,DampingAlpha=0.000000,MaxLinearVelocity=700.000000,MaxAngularVelocity=8.000000,MaxLinearAcceleration=2800.000000,MaxAngularAcceleration=32.000000,ExternalLinearDragV=(X=0.500000,Y=0.500000,Z=0.500000)))),Riding_Standing=(RigidBodyAnimNode=(SimSpaceSettings=(WorldAlpha=1.000000,VelocityScaleZ=0.500000,DampingAlpha=0.000000,MaxLinearVelocity=700.000000,MaxAngularVelocity=8.000000,MaxLinearAcceleration=2800.000000,MaxAngularAcceleration=32.000000,ExternalLinearDragV=(X=0.500000,Y=0.500000,Z=0.500000)))),Riding_Moving=(RigidBodyAnimNode=(SimSpaceSettings=(WorldAlpha=1.000000,VelocityScaleZ=0.500000,DampingAlpha=0.000000,MaxLinearVelocity=700.000000,MaxAngularVelocity=8.000000,MaxLinearAcceleration=2800.000000,MaxAngularAcceleration=32.000000,ExternalLinearDragV=(X=0.500000,Y=0.500000,Z=0.500000)))),Riding_Moving_Targeting=(RigidBodyAnimNode=(SimSpaceSettings=(WorldAlpha=1.000000,VelocityScaleZ=0.500000,DampingAlpha=0.000000,MaxLinearVelocity=700.000000,MaxAngularVelocity=8.000000,MaxLinearAcceleration=2800.000000,MaxAngularAcceleration=32.000000,ExternalLinearDragV=(X=0.500000,Y=0.500000,Z=0.500000)))),Tactical_Sprint=(RigidBodyAnimNode=(ComponentLinearAccScale=(X=0.300000,Y=0.000000,Z=0.000000),ComponentLinearVelScale=(X=0.300000,Y=0.000000,Z=0.000000),SimSpaceSettings=(WorldAlpha=1.000000,VelocityScaleZ=0.250000,DampingAlpha=0.000000,MaxLinearVelocity=500.000000,MaxAngularVelocity=6.000000,MaxLinearAcceleration=2000.000000,MaxAngularAcceleration=24.000000,ExternalLinearDragV=(X=0.200000,Y=0.200000,Z=0.200000)))),Rail_Grinding=(RigidBodyAnimNode=(SimSpaceSettings=(WorldAlpha=1.000000,VelocityScaleZ=0.500000,DampingAlpha=0.000000,MaxLinearVelocity=700.000000,MaxAngularVelocity=8.000000,MaxLinearAcceleration=2800.000000,MaxAngularAcceleration=32.000000,ExternalLinearDragV=(X=0.500000,Y=0.500000,Z=0.500000)))),)
 }
 RBAN Settings Out {
  type=FortRigidBodyAnimNodeParameters
  default {simSpaceSettings {
    dampingAlpha=0
    maxAngularAcceleration=16
    maxAngularVelocity=4
    maxLinearAcceleration=800
    maxLinearVelocity=200
    velocityScaleZ=0.5
    worldAlpha=1
   }}
 }
 SplineFunc {type=boolean}
]
